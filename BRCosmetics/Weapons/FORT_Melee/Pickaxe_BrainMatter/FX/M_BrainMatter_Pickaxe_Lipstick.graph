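# Translucent unlit Niagara sprite material for the BrainMatter pickaxe's lipstick effect, tinted by particle color.
# A gradient mask, sampled in polar-style coordinates and shifted by the Warp dynamic parameter, distorts the lipstick texture's UVs.
# That texture's green channel blends main and edge colors, blue shades both, and red times particle alpha forms an opacity mask.
# The Sheen dynamic parameter moves a rotated, sharpened sine band that strongly brightens the edge color.
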Material /BRCosmetics/Weapons/FORT_Melee/Pickaxe_BrainMatter/FX/M_BrainMatter_Pickaxe_Lipstick
graph {
 properties {bUsedWithNiagaraSprites=true, blendMode=BLEND_Translucent, pixelDepthOffsetMode=PDOM_Legacy, shadingModel=MSM_Unlit, usedShadingModels=Unlit}
 nodes [
  n1 = ParticleColor()
  n2 = VectorParameter() {parameterName=Color_Main, defaultValue=<rgba 0.4 0.016 0.0352 1>}
  n3 = Subtract(A=TextureCoordinate()) {constB=0.5}
  n4 = MaterialFunctionCall(Float2=@n3) {materialFunction=&BreakOutFloat2Components}
  n5 = Arctangent2Fast(Y=@n4.G, X=@n4.G)
  n6 = MaterialFunctionCall() {materialFunction=&Pi}
  n7 = Constant() {r=2}
  n8 = AppendVector(
   A=SquareRoot(None=MaterialFunctionCall(Vector 2=@n3) {materialFunction=/Engine/Functions/Engine_MaterialFunctions02/Utility/VectorLength}.V2 Length)
   B=Divide(
    A=Add(A=@n5, B=@n6.Result)
    B=MaterialFunctionCall(Multiplier=@n7) {materialFunction=&Pi}.Result
   )
  )
  n9 = DynamicParameter() {paramNames [Sheen,Warp,Param3,Param4]}
  n10 = AppendVector(A=@n9.Warp, B=Constant())
  n11 = Power(Base=TextureSampleParameter2D(UVs=Add(A=@n8, B=@n10)) {parameterName=Tex2, texture=/Game/Effects/Fort_Effects/Emotes/Sienna/T_Horizontal_gradient_Mask}.R) {constExponent=1}
  n12 = MaterialFunctionCall(Float2=TextureCoordinate()) {materialFunction=&BreakOutFloat2Components}
  n13 = Multiply(
   A=Multiply(A=Saturate(None=@n11)) {constB=0.5}
   B=AppendVector(A=LinearInterpolate(Alpha=@n12.R) {constA=-1}, B=LinearInterpolate(Alpha=@n12.G) {constA=-1})
  )
  n14 = TextureCoordinate()
  n15 = TextureSampleParameter2D(UVs=Saturate(None=Add(A=@n13, B=@n14))) {
   parameterName=Tex1
   samplerType=SAMPLERTYPE_LinearColor
   texture=&/T_BrainMatter_Lipstick
  }
  n16 = LinearInterpolate(A=Multiply(A=@n2.RGB) {constB=0.4}, B=@n2.RGB, Alpha=@n15.B)
  n17 = VectorParameter() {parameterName=Color_Edge, defaultValue=<rgba 0.64 0.2671 0.0512 1>}
  n18 = LinearInterpolate(A=Multiply(A=@n17.RGB) {constB=0.05}, B=@n17.RGB, Alpha=@n15.B)
  n19 = TextureCoordinate()
  n20 = ComponentMask(None=MaterialFunctionCall(UVs=Add(A=@n9.Sheen, B=@n19), Rotation Center=Constant() {r=0.5}, "Rotation Angle (0-1)"=Constant() {r=-0.1}) {materialFunction=&CustomRotator}.Rotated Values) {r=true}
  n21 = Power(Base=Sine(None=Multiply(A=@n20) {constB=0.5})) {constExponent=4}
  n22 = SmoothStep(Value=@n21) {constMin=0.9}
  n23 = SmoothStep(Value=@n21) {constMin=0.95}
  n24 = LinearInterpolate(A=@n18, B=Multiply(A=@n18) {constB=200}, Alpha=Saturate(None=LinearInterpolate(A=@n22, B=@n23) {constAlpha=0.8})) {constAlpha=0}
  MaterialFunctionCall(Emissive=Multiply(A=@n1.RGB, B=LinearInterpolate(A=@n16, B=@n24, Alpha=@n15.G))) {materialFunction=&MF_FX_EmissiveAdjustments}
  Multiply(A=@n1.A, B=@n15.R)
  TextureCoordinate()
 ]
}
